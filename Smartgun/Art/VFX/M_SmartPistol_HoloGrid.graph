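# Translucent, two-sided unlit VFX material for the Smartgun smart pistol's holographic targeting grid, usable on Niagara mesh particles and ribbons.
# Animated concentric rings and a tiled grid texture are masked by a sine-based fill that fades toward edges.
# Target Position, projected to clip space, offsets a target dot and the rings, gated by Has Target.
# Particle color tints emissive, while near-camera depth fade, inverse Fresnel and particle alpha shape opacity.
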Material /Smartgun/Art/VFX/M_SmartPistol_HoloGrid
graph {
 properties {
  bUsedWithNiagaraMeshParticles=true
  bUsedWithNiagaraRibbons=true
  blendMode=BLEND_Translucent
  pixelDepthOffsetMode=PDOM_Legacy
  shadingModel=MSM_Unlit
  twoSided=true
  usedShadingModels=Unlit
 }
 nodes [
  n1 = Multiply(A=Time()) {constB=0.25}
  n2 = Constant2Vector()
  n3 = Subtract(A=TextureCoordinate(), B=@n2) {constB=0.5}
  n4 = MaterialFunctionCall(Float2=Multiply(A=Cosine(None=@n1), B=@n3)) {materialFunction=&BreakOutFloat2Components}
  n5 = MaterialFunctionCall(Float2=Multiply(A=@n3, B=Sine(None=@n1))) {materialFunction=&BreakOutFloat2Components}
  Add(A=AppendVector(A=Subtract(A=@n4.R, B=@n5.G), B=Add(A=@n5.R, B=@n4.G)), B=@n2)
  Constant()
  n6 = Multiply(A=Time(), B=ScalarParameter() {defaultValue=2, parameterName=Circle Speed})
  n7 = Subtract(A=TextureCoordinate(), B=Multiply(A=NamedRerouteUsage(), B=ScalarParameter() {defaultValue=1, parameterName=Has Target}))
  n8 = Length(None=Multiply(A=ConstantBiasScale(None=@n7) {bias=-0.5, scale=2}, B=ScalarParameter() {defaultValue=3, parameterName=Circle Tile})) {is=class MaterialExpressionLength}
  n9 = Saturate(None=Sine(None=Add(A=@n6, B=@n8)))
  n10 = Power(Base=@n9) {constExponent=50}
  n11 = VectorParameter() {parameterName=GridScale, defaultValue=<rgba 2 2 0 0>}
  n12 = TextureSample(UVs=Multiply(A=ComponentMask(None=@n11.RGB) {g=true, r=true}, B=TextureCoordinate())) {samplerType=SAMPLERTYPE_LinearGrayscale, texture=/Game/Packages/Fortress_SharedMaterials/Textures/T_Grid__M}
  LinearInterpolate(A=@n10, B=Power(Base=@n9) {constExponent=3}, Alpha=@n12.R)
  n13 = Multiply(A=TextureCoordinate()) {constB=0.5}
  n14 = Multiply(A=Sine(None=ComponentMask(None=@n13) {r=true}), B=Sine(None=ComponentMask(None=@n13) {g=true}))
  n15 = Multiply(A=@n14, B=@n14)
  n16 = Add(
   A=Multiply(A=@n10, B=ScalarParameter() {defaultValue=0.3, parameterName=Circle Strength})
   B=Multiply(A=@n12.R, B=ScalarParameter() {defaultValue=0.025, parameterName=Grid Strength})
  )
  n17 = Constant() {r=1}
  n18 = MaterialFunctionCall(Float2=ViewSize()) {materialFunction=&BreakOutFloat2Components}
  n19 = Divide(A=@n18.G, B=@n18.G)
  n20 = Constant2Vector() {g=-0.5, r=0.5}
  n21 = MaterialFunctionCall(World Position=VectorParameter() {parameterName=Target Position}.RGB) {materialFunction=&TransformToClipSpace}
  n22 = Multiply(
   A=ScalarParameter() {defaultValue=0.25, parameterName=Target Grid Scale}
   B=Multiply(A=AppendVector(A=@n17, B=@n19) {desc=(1,0.5)}, B=Multiply(A=@n20, B=@n21.Clip Space XY))
  )
  n23 = SphereMask(
   A=TextureCoordinate()
   B=Add(A=NamedRerouteDeclaration(None=@n22) {name=Center To Target, nodeColor=<rgba 1 0 0.1406 1>}, B=Constant() {r=0.5})
   Radius=ScalarParameter() {defaultValue=0.1, parameterName=Target Scale}
   Hardness=LinearInterpolate(Alpha=ScalarParameter() {parameterName=Target Hardness}) {constB=100}
  ) {hardnessPercent=75}
  n24 = Multiply(A=Saturate(None=Power(Base=@n23) {constExponent=10}), B=ScalarParameter() {defaultValue=1, parameterName=Has Target})
  n25 = Add(
   A=Multiply(A=@n15, B=ScalarParameter() {defaultValue=0.25, parameterName=Fill Strength})
   B=Multiply(A=Power(Base=@n15) {constExponent=0.66}, B=Add(A=@n16, B=@n24))
  )
  n26 = ParticleColor()
  MaterialFunctionCall(Emissive=Multiply(A=LinearInterpolate(Alpha=@n25) {constA=0.25}, B=@n26.RGB)) {materialFunction=&MF_FX_EmissiveAdjustments}
  n27 = Cosine(None=Multiply(A=TextureCoordinate()) {constB=5})
  n28 = MaterialFunctionCall(Float2=Power(Base=Saturate(None=@n27)) {constExponent=100}) {materialFunction=&BreakOutFloat2Components}
  Max(A=@n28.R, B=@n28.R)
  n29 = PixelDepth()
  n30 = ScalarParameter() {defaultValue=150, group=Grid, parameterName=Near Camera Fade Start}
  n31 = Saturate(None=Divide(A=Subtract(A=@n29.R, B=@n30), B=ScalarParameter() {defaultValue=500, group=Grid, parameterName=Near Camera Fade}))
  n32 = Multiply(A=@n25, B=Multiply(A=ScalarParameter() {defaultValue=2, group=Grid, parameterName=Opacity Multiplier}, B=Clamp(None=@n31) {minDefault=0.1})) {constB=3}
  n33 = Fresnel(ExponentIn=ScalarParameter() {defaultValue=4, group=Grid, parameterName=Fresnel Exponent}) {baseReflectFraction=0}
  Multiply(A=Multiply(A=Saturate(None=@n32), B=OneMinus(None=@n33)), B=@n26.A)
  OneMinus(None=Fresnel(ExponentIn=ScalarParameter() {defaultValue=1, parameterName=ExponentIn}) {baseReflectFraction=0})
  ScalarParameter() {parameterName=Time}
  TextureSample() {samplerType=SAMPLERTYPE_LinearGrayscale, texture=&T_Grid__M}
 ]
}
