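# Niagara system for smart pistol aiming, with a glow mesh, looping flipbook cone glow and targeting reticle.
# User weapon index switches color, sizing and offsets, while system timelines animate the glow and cone.
# Two counter-rotating reticle sprites size and fade with target facing, with lock progress boosting their size pulse.
# A GPU reticle variant adds occlusion testing, distance fade and time-of-day brightness, and both hold screen size via camera-based scaling.
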NiagaraSystem /Smartgun/Art/VFX/NS_SmartPistol_Aiming
dependencies {
 usedDataInterfaces [
  &NiagaraDataInterfaceMeshRendererInfo
  &NiagaraDataInterfaceCurve
  &NiagaraDataInterfaceSpriteRendererInfo
  /Script/Niagara.NiagaraDataInterfaceOcclusion
  &NiagaraDataInterfaceCamera
  &NiagaraDataInterfaceArrayFloat
 ]
 usedDynamicInputs [
  &RemapRange
  &DotProduct
  &ReturnCameraVector
  &SetBoolByFloatComparison
  &MakeCustomFloatFromBool
 ]
 usedModules [
  &self:SystemSpawnScript.SystemScriptSource.SystemScriptGraph.NiagaraNodeAssignment_2.SetVariables_#1
  &SystemState
  &self:SystemSpawnScript.SystemScriptSource.SystemScriptGraph.NiagaraNodeAssignment_4.SetVariables_#2
  &self:SystemSpawnScript.SystemScriptSource.SystemScriptGraph.NiagaraNodeAssignment_0.SetVariables_#3
  &Timeline
  &self:SystemSpawnScript.SystemScriptSource.SystemScriptGraph.NiagaraNodeAssignment_1.SetVariables_#4
  &self:SystemSpawnScript.SystemScriptSource.SystemScriptGraph.NiagaraNodeAssignment_3.SetVariables_#5
 ]
}
emitters {
 Glow {
  emitterUpdateScript [
   &EmitterState(
    Inactive Response=::NewEnumerator1
    Life Cycle Mode=::NewEnumerator1
    Loop Behavior=::NewEnumerator1
    Loop Duration Mode=::NewEnumerator1
    UseLoopDelay=false
   )
   &SpawnBurst_Instantaneous(Use Spawn Probability=false, UseLoopCountLimit=true)
   /SetVariables(Emitter.Circle Strength=0.3) {isSetParameters=true}
   /SetVariables(Emitter.Circle Speed:NiagaraFloat=&Lerp_Float(A=-1, B=&Lerp_Float(A=2, B=0.75, Alpha=$System.Timeline001.NormalizedTime), Alpha=$System.HasTarget)) {isSetParameters=true}
   &SpawnBurst_Instantaneous(Spawn Time=4, Use Spawn Probability=false, UseLoopCountLimit=false) {module=SpawnBurst_Instantaneous001}
  ]
  particleSpawnScript [
   &InitializeParticle(
    Color=&MergeHelper_ColorSwitch(0=<rgba 0 1 2 1>, 1=<rgba 2 0.1667 0 1>, 2=?, 3=?, Index=$User.Weapon_Index, Num=::NewEnumerator0)
    Mesh Renderer Array Visibility Mode=::NewEnumerator1
    Mesh Renderer Info=DataInterface() {propertyValues {MeshRenderer=&self:Glow_0.NiagaraMeshRendererProperties_2}}
    Mesh Renderer Mesh Index:NiagaraInt32=$System.Weapon_Index
    Mesh Scale=&Multiply_VectorByFloat(Vector=<1 1 1>, Float=0.15)
    Mesh Scale Mode=::NewEnumerator4
    Position Offset=&Add_Vector(
     A=$System.LocationOffset
     B=&MergeHelper_VectorSwitch(0=<0 0 0>, 1=<15 0 0>, 2=?, 3=?, Index=$User.Weapon_Index, Num=::NewEnumerator0)
    )
    UsePositionOffset=true
    Write Material Random=true
   )
   &InitialMeshOrientation(Mesh Orientation Mode=::NewEnumerator2, Orientation Coordinate Space=::Local, Orientation Vector=<0 1 0>, Use Rotation Vector=false)
  ]
  particleUpdateScript [
   &ParticleState(Kill Particles When Lifetime Has Elapsed=false, Loop Particles Lifetime=true)
   &ScaleColor(
    Scale Alpha=&FloatFromCurve(
     FloatCurve=DataInterface() {propertyValues {Curve {keys [
         {arriveTangent=5.6222, interpMode=RCIM_Cubic, time=0, value=0}
         {interpMode=RCIM_Cubic, tangentMode=RCTM_User, time=0.5, value=1}
         {arriveTangent=-0.0405, interpMode=RCIM_Cubic, leaveTangent=-0.0405, tangentMode=RCTM_User, time=1, value=0.33}
        ]}}}
     CurveIndex=$System.Timeline.NormalizedTime
     Scale Curve=&Lerp_Float(A=1, B=2, Alpha=$System.Timeline001.NormalizedTime)
    )
    ScaleA=true
    ScaleRGB=false
   )
   &ScaleMeshSize(Scale Factor=&Multiply_VectorByFloat(
     Vector=&Lerp_Vector(A=<1 1 1>, B=<0.75 0.75 0.75>, Alpha=$System.Timeline001.NormalizedTime)
     Float=&FloatFromCurve(
      FloatCurve=DataInterface() {propertyValues {Curve {keys [
          {interpMode=RCIM_Cubic, tangentMode=RCTM_User, time=0, value=0.25}
          {interpMode=RCIM_Cubic, tangentMode=RCTM_User, time=0.66, value=1.2}
          {tangentMode=RCTM_User, time=1, value=1}
         ]}}}
      CurveIndex=$System.Timeline.NormalizedTime
     )
    ))
   /SetVariables(Particles.Position=&AddVectorToPosition(
     Position=$Particles.Initial.Position
     Vector=&Lerp_Vector(A=<0 0 0>, B=<25 0 0>, Alpha=$System.Timeline001.NormalizedTime)
    )) {isSetParameters=true}
  ]
  rendererClasses [&NiagaraMeshRendererProperties]
  renderers [{rendererClass=&NiagaraMeshRendererProperties, rendererIndex=0}]
 }
 SubUV_ConeGlowLoop {
  emitterUpdateScript [
   &EmitterState(
    Inactive Response=::NewEnumerator1
    Life Cycle Mode=::NewEnumerator1
    Loop Behavior=::NewEnumerator1
    Loop Duration Mode=::NewEnumerator1
    UseLoopDelay=false
   )
   &SpawnBurst_Instantaneous(Use Spawn Probability=false, UseLoopCountLimit=true)
   &SpawnRate(SpawnRate=2, Use Spawn Probability=false)
  ]
  particleSpawnScript [
   &InitializeParticle(
    Color=&MergeHelper_ColorSwitch(0=<rgba 0 1 2 1>, 1=<rgba 2 0.1667 0 1>, 2=?, 3=?, Index=$User.Weapon_Index, Num=::NewEnumerator0)
    Lifetime Max=1
    Lifetime Mode=::NewEnumerator1
    Position Offset=$System.LocationOffset
    Sprite Size=&MergeHelper_Vector2Switch(0=<25 25>, 1=<35 50>, 2=?, 3=?, 4=?, Index=$User.Weapon_Index, Num=::NewEnumerator0)
    Sprite Size Mode=::NewEnumerator4
    UsePositionOffset=true
    Write Material Random=true
   )
   &SpriteFacingAndAlignment(Sprite Alignment=<1 0 0>, Sprite Facing=<0 0 1>, WriteAlignment=true, WriteFacing=false)
   &SubUVAnimation(
    Evaluation Type=::NewEnumerator0
    "Sprite Renderer "=DataInterface() {propertyValues {SpriteRenderer=&self:SubUV_ConeGlowLoop_3.NiagaraSpriteRendererProperties_0}}
    SubUV Animation Mode:ENiagara_SubUVLookupModeV2=::NewEnumerator2
    UseEndFrame=false
    UseStartFrame=false
   )
  ]
  particleUpdateScript [
   &SolveForcesAndVelocity(Clamp Velocity=false, Limit Acceleration=false)
   &ParticleState()
   &ScaleSpriteSize(
    Scale Factor=&MakeVector2D(
     X=&FloatFromCurve(
      FloatCurve=DataInterface() {propertyValues {Curve {keys [{tangentMode=RCTM_User, time=0, value=0}, {tangentMode=RCTM_User, time=1, value=1}]}}}
      Scale Curve=&Lerp_Float(A=1, B=0.8, Alpha=$System.Timeline001.NormalizedTime)
     )
     Y=&Lerp_Float(A=1, B=1.5, Alpha=$System.Timeline001.NormalizedTime)
    )
    Scale Sprite Size Mode=::NewEnumerator2
   )
   &ScaleColor(
    Scale Alpha=&FloatFromCurve(
     FloatCurve=DataInterface() {propertyValues {Curve {keys [
         {arriveTangent=9.0725, interpMode=RCIM_Cubic, leaveTangent=9.0725, tangentMode=RCTM_User, time=0, value=0}
         <at 0.1408 0.9847 RCIM_Cubic>
         {arriveTangent=-0.0405, interpMode=RCIM_Cubic, leaveTangent=-0.0405, tangentMode=RCTM_User, time=1, value=0}
        ]}}}
     Scale Curve=&Lerp_Float(A=0.5, B=1, Alpha=$System.Timeline001.NormalizedTime)
    )
    ScaleA=true
    ScaleRGB=false
   )
   &CameraOffset(Camera Offset Amount=15)
  ]
  rendererClasses [&NiagaraSpriteRendererProperties]
  renderers [{rendererClass=&NiagaraSpriteRendererProperties}]
 }
 Targeted_Reticle {
  emitterUpdateScript [
   &EmitterState(
    Inactive Response=::NewEnumerator1
    Life Cycle Mode=::NewEnumerator1
    Loop Behavior=::NewEnumerator1
    Loop Duration Mode=::NewEnumerator1
    UseLoopDelay=false
   )
   &SpawnBurst_Instantaneous(Spawn Count=2, Use Spawn Probability=false, UseLoopCountLimit=true)
  ]
  particleSpawnScript [
   /SetVariables(Particles.Visibility_wide:NiagaraInt32=&MakeIntFromFloat(Float=&ReturnNormalizedExecIndex())) {isSetParameters=true, module=SetVariables_#6}
   &InitializeParticle(
    Color=&MergeHelper_ColorSwitch(0=<rgba 0 1.5 3 1>, 1=<rgba 3 0.25 0 1>, 2=?, 3=?, Index=$User.Weapon_Index, Num=::NewEnumerator0)
    Lifetime=0.25
    Position Offset=$System.LocationOffset
    Sprite Size Mode=::NewEnumerator3
    Uniform Sprite Size=120
    UsePositionOffset=true
    Write Material Random=true
   )
   /SetVariables() {isSetParameters=true, module=SetVariables_#7}
  ]
  particleUpdateScript [
   &ParticleState(Kill Particles When Lifetime Has Elapsed=false, Loop Particles Lifetime=true)
   &SpriteRotationRate(Rotation Rate=&Lerp_Float(A=-30, B=30, Alpha=/Niagara/DynamicInputs/Math/Floor_Float(Float:NiagaraFloat=&ReturnNormalizedExecIndex())))
   /Niagara/Modules/Update/Camera/MaintainInCameraParticleScale(
    Pixel Depth:NiagaraFloat=&Add_Float(A=$Engine.Owner.LODDistance, B=100)
    Render Target Width:NiagaraFloat=&Multiply_Float(A=2560, B=1)
   )
   /SetVariables(Particles.SpriteSize=/Niagara/DynamicInputs/Add/Add_Vector2(
     A:Vector2f=&Multiply_Vector2D_ByFloat(
      Float=&FloatFromCurve(
       FloatCurve=DataInterface() {propertyValues {Curve {keys [
           {arriveTangent=4.1785, interpMode=RCIM_Cubic, time=0, value=0}
           {arriveTangent=1.1179, interpMode=RCIM_Cubic, leaveTangent=1.1179, tangentMode=RCTM_User, time=0.2042, value=0.6228}
           {arriveTangent=0.0224, interpMode=RCIM_Cubic, leaveTangent=0.0224, tangentMode=RCTM_User, time=1, value=1}
          ]}}}
       CurveIndex=$System.TargetFacing
      )
      Vector2D=$Particles.SpriteSize
     )
     B:Vector2f=&Vector2DFromFloat(Value=&FloatFromCurve(
       FloatCurve=DataInterface() {propertyValues {Curve {keys [
           {interpMode=RCIM_Cubic, tangentMode=RCTM_User, time=0, value=0}
           {interpMode=RCIM_Cubic, tangentMode=RCTM_User, time=0.5, value=1}
           {tangentMode=RCTM_User, time=1, value=0}
          ]}}}
       Scale Curve=&Multiply_Float(A=$System.TargetFacing, B=&Lerp_Float(A=5, B=15, Alpha=$User.LockProgress))
      ))
    )) {isSetParameters=true, module=SetVariables_#8}
   /SetVariables(Particles.SpriteSize_Wide:Vector2f=&Multiply_Vector2D(A=$Particles.SpriteSize, B=<2 0.5>)) {isSetParameters=true, module=SetVariables_#9}
   &ScaleColor(
    Scale Alpha=&FloatFromCurve(FloatCurve=DataInterface() {propertyValues {Curve {keys [<at 0 0 RCIM_Cubic>,<at 0.5 1 RCIM_Cubic>]}}}, CurveIndex=$System.TargetFacing)
    Scale RGB=&VectorFromFloat(Value=&Lerp_Float(A=2, B=2, Alpha=$User.LockProgress))
    ScaleA=true
    ScaleRGB=true
   )
   &CameraOffset(Camera Offset Amount=5)
   /SetVariables(Particles.ColorGlow:LinearColor=&MultiplyLinearColor(
     Linear Color=$Particles.Initial.Color
     Scale Factor=&MakeFloat4FromVectorAndFloat("Vector (RGB)"=<1 1 1>, "Float (Alpha)"=0.2)
    )) {isSetParameters=true, module=SetVariables_#10}
  ]
  rendererClasses [&NiagaraSpriteRendererProperties]
  renderers [
   {rendererClass=&NiagaraSpriteRendererProperties}
   {rendererClass=&NiagaraSpriteRendererProperties, rendererIndex=1}
  ]
 }
 Targeted_Reticle_Occlusion {
  simTarget=GPUComputeSim
  emitterUpdateScript [
   &EmitterState(
    Inactive Response=::NewEnumerator1
    Life Cycle Mode=::NewEnumerator1
    Loop Behavior=::NewEnumerator1
    Loop Duration Mode=::NewEnumerator1
    UseLoopDelay=false
   )
   &SpawnBurst_Instantaneous(Spawn Count=2, Use Spawn Probability=false, UseLoopCountLimit=true)
  ]
  particleSpawnScript [
   /SetVariables(Particles.Visibility_wide:NiagaraInt32=&MakeIntFromFloat(Float=&ReturnNormalizedExecIndex())) {isSetParameters=true, module=SetVariables_#6}
   &InitializeParticle(
    Color=&MergeHelper_ColorSwitch(0=<rgba 0 1.5 3 1>, 1=<rgba 3 0.25 0 1>, 2=?, 3=?, Index=$User.Weapon_Index, Num=::NewEnumerator0)
    Lifetime=0.25
    Position Offset=$System.LocationOffset
    Sprite Size Mode=::NewEnumerator3
    Uniform Sprite Size=120
    UsePositionOffset=true
    Write Material Random=true
   )
   /SetVariables() {isSetParameters=true, module=SetVariables_#7}
  ]
  particleUpdateScript [
   &ParticleState(Kill Particles When Lifetime Has Elapsed=false, Loop Particles Lifetime=true)
   &self:Flare_Occlusion(OcclusionSize=20) {module=Flare_Occlusion}
   &SpriteRotationRate(Rotation Rate=&Lerp_Float(A=-30, B=30, Alpha=&Floor_Float(Float:NiagaraFloat=&ReturnNormalizedExecIndex())))
   &MaintainInCameraParticleScale(
    Pixel Depth:NiagaraFloat=&Add_Float(A=$Engine.Owner.LODDistance, B=100)
    Render Target Width:NiagaraFloat=&Multiply_Float(A=2560, B=1)
   )
   /SetVariables(Particles.SpriteSize=&Add_Vector2(
     A:Vector2f=&Multiply_Vector2D_ByFloat(
      Float=&FloatFromCurve(
       FloatCurve=DataInterface() {propertyValues {Curve {keys [
           {arriveTangent=4.1785, interpMode=RCIM_Cubic, time=0, value=0}
           {arriveTangent=1.1179, interpMode=RCIM_Cubic, leaveTangent=1.1179, tangentMode=RCTM_User, time=0.2042, value=0.6228}
           {arriveTangent=0.0224, interpMode=RCIM_Cubic, leaveTangent=0.0224, tangentMode=RCTM_User, time=1, value=1}
          ]}}}
       CurveIndex=$System.TargetFacing
      )
      Vector2D=$Particles.SpriteSize
     )
     B:Vector2f=&Vector2DFromFloat(Value=&FloatFromCurve(
       FloatCurve=DataInterface() {propertyValues {Curve {keys [
           {interpMode=RCIM_Cubic, tangentMode=RCTM_User, time=0, value=0}
           {interpMode=RCIM_Cubic, tangentMode=RCTM_User, time=0.5, value=1}
           {tangentMode=RCTM_User, time=1, value=0}
          ]}}}
       Scale Curve=&Multiply_Float(A=$System.TargetFacing, B=&Lerp_Float(A=5, B=15, Alpha=$User.LockProgress))
      ))
    )) {isSetParameters=true, module=SetVariables_#8}
   /SetVariables(Particles.SpriteSize_Wide:Vector2f=&Multiply_Vector2D(A=$Particles.SpriteSize, B=<2 0.5>)) {isSetParameters=true, module=SetVariables_#9}
   &ScaleColor(
    Scale Alpha=&FloatFromCurve(
     FloatCurve=DataInterface() {propertyValues {Curve {keys [<at 0 0 RCIM_Cubic>,<at 0.5 1 RCIM_Cubic>]}}}
     CurveIndex=$System.TargetFacing
     Scale Curve=&Multiply_Float(A=$Particles.Visibility, B=$System.DistanceNormalized)
    )
    Scale RGB=&VectorFromFloat(Value=&FNTimeOfDayScalar(Night Scale=0.4, Morning Scale=1, Day Scale=2, Evening Scale=1))
    ScaleA=true
    ScaleRGB=true
   )
   &CameraOffset(Camera Offset Amount=5)
   /SetVariables(Particles.ColorGlow:LinearColor=&MultiplyLinearColor(
     Linear Color=$Particles.Initial.Color
     Scale Factor=&MakeFloat4FromVectorAndFloat(
      "Vector (RGB)"=&VectorFromFloat(Value=&FNTimeOfDayScalar(Night Scale=0.25, Morning Scale=0.5, Evening Scale=0.5))
      "Float (Alpha)"=&FloatFromCurve(
       FloatCurve=DataInterface() {propertyValues {Curve {keys [<at 0 0 RCIM_Cubic>,<at 0.5 1 RCIM_Cubic>]}}}
       CurveIndex=$System.TargetFacing
       Scale Curve=&Multiply_Float(A=0.25, B=&Multiply_Float(A=$Particles.Visibility, B=$System.DistanceNormalized))
      )
     )
    )) {isSetParameters=true, module=SetVariables_#10}
  ]
  rendererClasses [&NiagaraSpriteRendererProperties]
  renderers [
   {rendererClass=&NiagaraSpriteRendererProperties}
   {rendererClass=&NiagaraSpriteRendererProperties, rendererIndex=1}
  ]
 }
}
summary {
 systemName=NS_SmartPistol_Aiming
 emitters [
  {emitterName=Glow, rendererClasses [&NiagaraMeshRendererProperties]}
  {emitterName=SubUV_ConeGlowLoop, rendererClasses [&NiagaraSpriteRendererProperties]}
  {emitterName=Targeted_Reticle_Occlusion, simTarget=GPUComputeSim, rendererClasses [&NiagaraSpriteRendererProperties]}
  {emitterName=Targeted_Reticle, rendererClasses [&NiagaraSpriteRendererProperties]}
 ]
 userVariables [
  User.HasTarget {type=NiagaraBool}
  User.LockProgress {type=NiagaraFloat, defaultValue=0, description=""}
  User.TargetPositionV {type=Vector3f, description="", defaultValue {struct=Vector3f, value=<0 0 0>}}
  User.Weapon_Index {type=NiagaraInt32, defaultValue=1}
 ]
}
